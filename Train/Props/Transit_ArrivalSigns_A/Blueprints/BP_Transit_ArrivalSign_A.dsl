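; Train arrival sign prop built on FortTrainTransitSign, reacting to arrival timing for its station.
; On BeginPlay it sets the parent's TickLocation to a point offset sideways and upward from the static mesh's world position.
; When the reported time until arrival is nearly zero, it raises a custom primitive data float on the mesh for ten seconds, then clears it.
; That same mesh is returned as the timer mesh.
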(event EventBeginPlay
  (|CallParentFunction)
  (Variables|FortTrainTransitSign|SetTickLocation (Math|Vector|MakeVector (+ (.x (Transformation|GetWorldLocation (Variables|Default|GetStaticMesh))) 95.0) (+ (.y (Transformation|GetWorldLocation (Variables|Default|GetStaticMesh))) 0) (+ (.z (Transformation|GetWorldLocation (Variables|Default|GetStaticMesh))) 4.2e+02))))

(event EventOnArrivalTimeIntervalReachedForStation
  (params (TImeUntilArrival :type Float%20%28double-precision%29))
  (bind _staticmesh (Variables|Default|GetStaticMesh))
  (if (Math|Float|NearlyEqual(Float) TImeUntilArrival)
    (Rendering|Material|SetCustomPrimitiveDataFloat _staticmesh 1 1.0)
    (Utilities|FlowControl|Delay 1e+01)
    (Rendering|Material|SetCustomPrimitiveDataFloat _staticmesh 1 0.0)))

(fn GetTimerMesh
  (return (Variables|Default|GetStaticMesh)))

(fn ConstructionScript
  (|CallParentFunction))
