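# MF_UV01_Mask is a material function in the Transit_ArrivalSigns_A prop folder that outputs a mask for UVs inside the 0–1 square.
# It splits the UV_Coord input into R and G and runs each through two opposing Step nodes against their default bounds.
# Multiplying the four results yields a branchless binary Mask: one inside the range, zero outside.
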MaterialFunction /Train/Props/Transit_ArrivalSigns_A/Materials/MF_UV01_Mask
graph {nodes [
  n1 = FunctionInput() {inputName=UV_Coord, inputType=FunctionInput_Vector2}
  n2 = ComponentMask(None=@n1) {r=true}
  n3 = Step(Y=@n2)
  n4 = Step(X=@n2)
  n5 = ComponentMask(None=@n1) {g=true}
  n6 = Step(Y=@n5)
  n7 = Step(X=@n5)
  FunctionOutput(None=Multiply(A=Multiply(A=@n3, B=@n4), B=Multiply(A=@n6, B=@n7))) {outputName=Mask}
 ]}
