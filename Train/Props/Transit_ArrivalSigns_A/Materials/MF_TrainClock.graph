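# Material function that renders a minutes-and-seconds countdown for train arrival signs onto incoming material attributes.
# Remaining time comes from custom primitive data, or game time via a debug switch, and drives FlipBook digits and a colon placed by Custom UV remapping.
# TrainArrived primitive data blends the countdown into a sign graphic.
# The mask drives time-of-day, exposure-compensated emissive with Fresnel, while an optional UV1 layer mask limits emissive, roughness and base-color changes.
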MaterialFunction /Train/Props/Transit_ArrivalSigns_A/Materials/MF_TrainClock
graph {nodes [
  n1 = NamedRerouteDeclaration(None=ScalarParameter() {defaultValue=30.5, desc=The time when next train will arrive, group=Clock, parameterName=ArrivalGapinMinutes}) {name=MaxMinutes, nodeColor=<rgba 1 0.7734 0 1>}
  n2 = Subtract(
   A=Add(
    A=ScalarParameter() {desc="When Train Arrives, set this time to be current game time in seconds.", group=Clock, parameterName=ArrivalTime}
    B=Multiply(A=@n1) {constB=60}
   )
   B=Time()
  )
  n3 = ScalarParameter() {bUseCustomPrimitiveData=true, group=Clock, parameterName=TimeRemaingFloat}
  n4 = NamedRerouteDeclaration(None=Max(A=StaticSwitchParameter(True=@n2, False=@n3) {group=Clock, parameterName=DebugTime}) {constB=0, desc=Keep Positive}) {name=SecondsOverallRemain, nodeColor=<rgba 0 1 0.2734 1>}
  n5 = NamedRerouteDeclaration(None=Divide(A=@n4) {constB=60}) {name=MinutesRemain, nodeColor=<rgba 0.2188 0 1 1>}
  n6 = Constant() {r=10}
  n7 = Divide(A=@n5, B=@n6)
  Floor(None=@n7)
  n8 = GetMaterialAttributes(None=FunctionInput() {inputName=MA, inputType=FunctionInput_MaterialAttributes}) {attributeGetTypes [#1, #2, #3]}
  n9 = ScalarParameter() {defaultValue=4, group=Clock, parameterName=EmissiveIntensity}
  n10 = Divide(A=@n9)
  n11 = Multiply(
   A=Multiply(A=VectorParameter() {group=Clock, parameterName=NumberColor, defaultValue=<rgba 1 1 1 1>}.RGB, B=NamedRerouteUsage())
   B=MaterialFunctionCall(Value_Morning=@n9, Value_Day=@n9, Value_Evening=@n10, Value_Night=@n10) {materialFunction=&Scalar_Time_of_Day}.Result
  )
  n12 = EyeAdaptation()
  n13 = Multiply(A=Fresnel(), B=VectorParameter() {group=Clock, parameterName=FresnelColor, defaultValue=<rgba 0.6354 0.8952 1 1>}.RGB)
  n14 = ScalarParameter() {defaultValue=0.2, group=Clock, parameterName=FresnelMulti}
  n15 = ComponentMask(None=TextureCoordinate() {coordinateIndex=1}) {r=true}
  n16 = ScalarParameter() {defaultValue=1, group=Clock, parameterName=LayerIndex}
  n17 = Constant()
  n18 = Constant() {r=1}
  n19 = Multiply(
   A=If(A=@n15, B=@n16, "A > B"=Constant() {r=1}, "A == B"=@n17, "A < B"=@n17)
   B=If(A=@n15, B=Add(A=@n16), "A > B"=Constant(), "A == B"=@n18, "A < B"=@n18)
  )
  n20 = Multiply(
   A=Add(A=Divide(A=@n11, B=@n12.EyeAdaptation), B=Multiply(A=@n13, B=@n14) {constB=0.2})
   B=StaticSwitchParameter(True=NamedRerouteDeclaration(None=@n19) {name=LayerMask, nodeColor=<rgba 0.2891 1 0 1>}, False=Constant() {r=1}) {group=Clock, parameterName=LayerSupport}
  )
  n21 = NamedRerouteUsage()
  n22 = Multiply(A=@n21, B=ScalarParameter() {defaultValue=1, group=Clock, parameterName=LayerRoughnessOverride}) {constB=0.2}
  n23 = StaticSwitchParameter(True=NamedRerouteUsage(), False=Constant()) {group=Clock, parameterName=LayerSupport}
  n24 = NamedRerouteUsage()
  n25 = LinearInterpolate(
   A=Multiply(A=@n24, B=ScalarParameter() {defaultValue=1, group=Clock, parameterName=LayerBaseColorOverride}) {constB=0.2}
   B=ScalarParameter() {defaultValue=0.05, group=Clock, parameterName=BaseColorArrivalOverride}
   Alpha=NamedRerouteUsage()
  )
  FunctionOutput(None=SetMaterialAttributes!(
    MaterialAttributes=@n8.MaterialAttributes
    EmissiveColor=Add(A=@n8.EmissiveColor, B=@n20)
    Roughness=LinearInterpolate(A=@n21, B=@n22, Alpha=@n23) {constB=0.3}
    BaseColor=LinearInterpolate(A=@n24, B=@n25, Alpha=@n23) {constB=0.3}
   ) {attributeSetTypes [#1, #2, #3]})
  n26 = VertexInterpolator(VS=PreSkinnedPosition())
  n27 = ScalarParameter() {defaultValue=0.01, group=Clock, parameterName=PreskinnedTilingScale}
  n28 = MaterialFunctionCall(Float2=StaticComponentMaskParameter(None=Multiply(A=@n26.PS, B=@n27)) {defaultG=true, defaultR=true, group=Clock, parameterName=PreskinnedProjection}) {materialFunction=&BreakOutFloat2Components}
  n29 = StaticComponentMaskParameter(None=MaterialFunctionCall() {materialFunction=&ObjectScale}.Scale XYZ) {defaultG=true, defaultR=true, group=Clock, parameterName=PreskinnedProjection}
  n30 = Multiply(A=@n28.R, B=Divide(A=ComponentMask(None=@n29) {r=true}, B=ComponentMask(None=@n29) {g=true}))
  n31 = ScalarParameter() {group=Clock, parameterName=Use UV Channel 02}
  n32 = LinearInterpolate(A=TextureCoordinate(), B=TextureCoordinate() {coordinateIndex=1}, Alpha=Clamp(None=Ceil(None=@n31)))
  NamedRerouteDeclaration(None=StaticSwitchParameter(
    True=MaterialFunctionCall(X=@n30, Y=@n28.G) {materialFunction=&MakeFloat2}.Result
    False=NamedRerouteDeclaration(None=@n32) {name=ProjectUV, nodeColor=<rgba 0.2422 1 0 1>}
   ) {group=Clock, parameterName=UsePreskinnedPos}) {name=UV_Number, nodeColor=<rgba 0.0781 1 0 1>}
  NamedRerouteDeclaration(None=@n8.BaseColor) {name=BaseColorIN, nodeColor=<rgba 1 0 0.2812 1>}
  n33 = Fmod(A=@n5, B=@n6) {desc=10X}
  NamedRerouteDeclaration(None=Divide(A=NamedRerouteDeclaration(None=@n33) {name=Single_Minute, nodeColor=<rgba 0.0547 1 0 1>}) {constB=10}) {name=MinuteProgressF2, nodeColor=<rgba 0.7344 1 0 1>}
  n34 = ScalarParameter() {defaultValue=0.65, group=Clock, parameterName=SmoothStepMin}
  n35 = NamedRerouteDeclaration(None=@n34) {name=StepMin, nodeColor=<rgba 0.8281 1 0 1>}
  n36 = Add(
   A=@n34
   B=NamedRerouteDeclaration(None=ScalarParameter() {defaultValue=0.2, group=Clock, parameterName=SmoothStepGap}) {name=StepGap, nodeColor=<rgba 0.875 0 1 1>}
  ) {constB=0.1}
  MaterialExpressionAdd_7 = Add(A=Multiply(A=NamedRerouteUsage()) {constB=4}, B=NamedRerouteUsage())
  MaterialExpressionNamedRerouteUsage_25 = NamedRerouteUsage()
  MaterialExpressionNamedRerouteUsage_0 = NamedRerouteUsage()
  n37 = Custom!(UVPoint=@MaterialExpressionAdd_7, UVScale=@MaterialExpressionNamedRerouteUsage_25, OriginalUV=@MaterialExpressionNamedRerouteUsage_0) {
   code="float2 DistUV =OriginalUV-UVPoint;\r\nDistUV = DistUV/UVScale + float2(0.5,0.5);\r\n\r\nreturn DistUV;"
   description=CopyToPointUV
   outputType=CMOT_Float2
   inputs [
    {inputName=UVPoint, input {expression=&self:MaterialExpressionAdd_7}}
    {inputName=UVScale, input {expression=&self:MaterialExpressionNamedRerouteUsage_25}}
    {inputName=OriginalUV, input {expression=&self:MaterialExpressionNamedRerouteUsage_0}}
   ]
  }
  n38 = MaterialFunctionCall(
   "Animation  Phase (0-1)"=LinearInterpolate(B=Divide() {constA=6, constB=16}, Alpha=NamedRerouteUsage()) {constB=0}
   Number of Rows=Constant() {r=4}
   Number of Columns=Constant() {r=4}
   Texture=NamedRerouteUsage()
   UVs=@n37
  ) {materialFunction=&FlipBook}
  n39 = MaterialFunctionCall!(UV_Coord=@n37) {materialFunction=&/MF_UV01_Mask}
  n40 = SmoothStep(
   Min=@n35
   Max=@n36
   Value=ChannelMaskParameter(None=Multiply(A=@n38.Result, B=@n39.Mask)) {group=Clock, parameterName=NumberChannel, sortPriority=1}.R
  ) {constMax=0.5, constMin=0.4}
  MaterialExpressionAdd_4 = Add(A=Multiply(A=NamedRerouteUsage()) {constB=3}, B=NamedRerouteUsage())
  MaterialExpressionNamedRerouteUsage_14 = NamedRerouteUsage()
  MaterialExpressionNamedRerouteUsage_1 = NamedRerouteUsage()
  n41 = Custom(UVPoint=@MaterialExpressionAdd_4, UVScale=@MaterialExpressionNamedRerouteUsage_14, OriginalUV=@MaterialExpressionNamedRerouteUsage_1) {
   code="float2 DistUV =OriginalUV-UVPoint;\r\nDistUV = DistUV/UVScale + float2(0.5,0.5);\r\n\r\nreturn DistUV;"
   description=CopyToPointUV
   outputType=CMOT_Float2
   inputs [
    {inputName=UVPoint, input {expression=&self:MaterialExpressionAdd_4}}
    {inputName=UVScale, input {expression=&self:MaterialExpressionNamedRerouteUsage_14}}
    {inputName=OriginalUV, input {expression=&self:MaterialExpressionNamedRerouteUsage_1}}
   ]
  }
  n42 = MaterialFunctionCall(
   "Animation  Phase (0-1)"=LinearInterpolate(B=NamedRerouteUsage(), Alpha=NamedRerouteUsage()) {constB=0}
   Number of Rows=Constant() {r=4}
   Number of Columns=Constant() {r=4}
   Texture=NamedRerouteUsage()
   UVs=@n41
  ) {materialFunction=&FlipBook}
  n43 = MaterialFunctionCall(UV_Coord=@n41) {materialFunction=&MF_UV01_Mask}
  n44 = SmoothStep(
   Min=@n35
   Max=@n36
   Value=ChannelMaskParameter(None=Multiply(A=@n42.Result, B=@n43.Mask)) {group=Clock, parameterName=NumberChannel, sortPriority=1}.R
  ) {constMax=0.5, constMin=0.4}
  n45 = NamedRerouteUsage()
  n46 = Add(A=@n45, B=NamedRerouteUsage()) {constB=0.1}
  n47 = LinearInterpolate(B=Divide() {constA=6, constB=16}, Alpha=NamedRerouteUsage()) {constB=0}
  n48 = Constant() {r=4}
  n49 = Constant() {r=4}
  n50 = NamedRerouteUsage()
  n51 = VectorParameter() {group=Clock, parameterName=LeftUVOffset, defaultValue=<rgba -0.06 0 0 1>}
  MaterialExpressionAdd_0 = Add(
   A=NamedRerouteDeclaration(None=ComponentMask(None=@n51.RGB) {g=true, r=true}) {name=OffsetUnit, nodeColor=<rgba 1 0 0.8203 1>}
   B=NamedRerouteUsage()
  )
  MaterialExpressionNamedRerouteUsage_6 = NamedRerouteUsage()
  MaterialExpressionNamedRerouteUsage_3 = NamedRerouteUsage()
  n52 = Custom(UVPoint=@MaterialExpressionAdd_0, UVScale=@MaterialExpressionNamedRerouteUsage_6, OriginalUV=@MaterialExpressionNamedRerouteUsage_3) {
   code="float2 DistUV =OriginalUV-UVPoint;\r\nDistUV = DistUV/UVScale + float2(0.5,0.5);\r\n\r\nreturn DistUV;"
   description=CopyToPointUV
   outputType=CMOT_Float2
   inputs [
    {inputName=UVPoint, input {expression=&self:MaterialExpressionAdd_0}}
    {inputName=UVScale, input {expression=&self:MaterialExpressionNamedRerouteUsage_6}}
    {inputName=OriginalUV, input {expression=&self:MaterialExpressionNamedRerouteUsage_3}}
   ]
  }
  n53 = ChannelMaskParameter(None=Multiply(
    A=MaterialFunctionCall("Animation  Phase (0-1)"=@n47, Number of Rows=@n48, Number of Columns=@n49, Texture=@n50, UVs=@n52) {materialFunction=&FlipBook}.Result
    B=MaterialFunctionCall(UV_Coord=@n52) {materialFunction=&MF_UV01_Mask}.Mask
   )) {group=Clock, parameterName=NumberChannel, sortPriority=1}
  n54 = NamedRerouteUsage()
  n55 = NamedRerouteUsage()
  n56 = NamedRerouteDeclaration(None=Divide() {constA=10, constB=16}) {name=ConstFrame, nodeColor=<rgba 1 0.7266 0 1>}
  n57 = NamedRerouteUsage()
  n58 = VectorParameter() {group=Clock, parameterName=TheRightNumber, channelNames {a=ScaleY, b=ScaleX, g=CenterY, r=CenterX}, defaultValue=<rgba 0 0 0.1 0.1>}
  MaterialExpressionNamedRerouteDeclaration_10 = NamedRerouteDeclaration(None=AppendVector(A=@n58.CenterX, B=@n58.CenterX)) {name=RightCornerUV, nodeColor=<rgba 0 0.7188 1 1>}
  MaterialExpressionNamedRerouteDeclaration_11 = NamedRerouteDeclaration(None=AppendVector(A=@n58.ScaleX, B=@n58.ScaleX)) {name=UVScale, nodeColor=<rgba 1 0.3281 0 1>}
  MaterialExpressionNamedRerouteUsage_4 = NamedRerouteUsage()
  n59 = Custom(
   UVPoint=@MaterialExpressionNamedRerouteDeclaration_10
   UVScale=@MaterialExpressionNamedRerouteDeclaration_11
   OriginalUV=@MaterialExpressionNamedRerouteUsage_4
  ) {
   code="float2 DistUV =OriginalUV-UVPoint;\r\nDistUV = DistUV/UVScale + float2(0.5,0.5);\r\n\r\nreturn DistUV;"
   description=CopyToPointUV
   outputType=CMOT_Float2
   inputs [
    {
     inputName=UVPoint
     input {expression=&self:MaterialExpressionNamedRerouteDeclaration_10}
    }
    {
     inputName=UVScale
     input {expression=&self:MaterialExpressionNamedRerouteDeclaration_11}
    }
    {inputName=OriginalUV, input {expression=&self:MaterialExpressionNamedRerouteUsage_4}}
   ]
  }
  n60 = Multiply(
   A=MaterialFunctionCall(
    "Animation  Phase (0-1)"=LinearInterpolate(B=@n56, Alpha=@n57) {constB=0}
    Number of Rows=Constant() {r=4}
    Number of Columns=Constant() {r=4}
    Texture=NamedRerouteUsage()
    UVs=NamedRerouteDeclaration(None=@n59) {name=TestUV, nodeColor=<rgba 0 0.8125 1 1>}
   ) {materialFunction=&FlipBook}.Result
   B=MaterialFunctionCall(UV_Coord=@n59) {materialFunction=&MF_UV01_Mask}.Mask
  )
  n61 = Add(
   A=Add(A=Add(A=@n40, B=@n44), B=SmoothStep(Min=@n45, Max=@n46, Value=@n53.R) {constMax=0.5, constMin=0.4})
   B=SmoothStep(
    Min=@n54
    Max=Add(A=@n54, B=@n55) {constB=0.1}
    Value=ChannelMaskParameter(None=@n60) {group=Clock, parameterName=NumberChannel, sortPriority=1}.R
   ) {constMax=0.5, constMin=0.4}
  )
  n62 = NamedRerouteUsage()
  MaterialExpressionAdd_2 = Add(A=Multiply(A=NamedRerouteUsage()) {constB=2}, B=NamedRerouteUsage())
  MaterialExpressionMultiply_5 = Multiply(A=NamedRerouteUsage()) {constB=4}
  MaterialExpressionNamedRerouteUsage_11 = NamedRerouteUsage()
  n63 = Custom(UVPoint=@MaterialExpressionAdd_2, UVScale=@MaterialExpressionMultiply_5, OriginalUV=@MaterialExpressionNamedRerouteUsage_11) {
   code="float2 DistUV =OriginalUV-UVPoint;\r\nDistUV = DistUV/UVScale + float2(0.5,0.5);\r\n\r\nreturn DistUV;"
   description=CopyToPointUV
   outputType=CMOT_Float2
   inputs [
    {inputName=UVPoint, input {expression=&self:MaterialExpressionAdd_2}}
    {inputName=UVScale, input {expression=&self:MaterialExpressionMultiply_5}}
    {inputName=OriginalUV, input {expression=&self:MaterialExpressionNamedRerouteUsage_11}}
   ]
  }
  n64 = NamedRerouteUsage()
  n65 = ChannelMaskParameter(None=Multiply(
    A=TextureSample(UVs=@n63, Tex=@n64) {samplerType=SAMPLERTYPE_Masks}.RGB
    B=MaterialFunctionCall(UV_Coord=@n63) {materialFunction=&MF_UV01_Mask}.Mask
   )) {group=Clock, maskChannel=Green, parameterName=SemicolonChannel, sortPriority=2, defaultValue=<rgba 0 1 0 0>}
  n66 = SmoothStep(Min=@n62, Max=Add(A=@n62, B=NamedRerouteUsage()) {constB=0.1}, Value=@n65.R) {constMax=0.5, constMin=0.4}
  n67 = NamedRerouteUsage()
  n68 = Add(A=@n67, B=NamedRerouteUsage()) {constB=0.1}
  MaterialExpressionAdd_22 = Add(A=Multiply(A=NamedRerouteUsage()) {constB=2}, B=NamedRerouteUsage())
  MaterialExpressionMultiply_18 = Multiply(A=NamedRerouteUsage()) {constB=2}
  MaterialExpressionNamedRerouteUsage_37 = NamedRerouteUsage()
  n69 = Custom(UVPoint=@MaterialExpressionAdd_22, UVScale=@MaterialExpressionMultiply_18, OriginalUV=@MaterialExpressionNamedRerouteUsage_37) {
   code="float2 DistUV =OriginalUV-UVPoint;\r\nDistUV = DistUV/UVScale + float2(0.5,0.5);\r\n\r\nreturn DistUV;"
   description=CopyToPointUV
   outputType=CMOT_Float2
   inputs [
    {inputName=UVPoint, input {expression=&self:MaterialExpressionAdd_22}}
    {inputName=UVScale, input {expression=&self:MaterialExpressionMultiply_18}}
    {inputName=OriginalUV, input {expression=&self:MaterialExpressionNamedRerouteUsage_37}}
   ]
  }
  n70 = NamedRerouteUsage()
  n71 = ChannelMaskParameter(None=Multiply(
    A=TextureSample(UVs=@n69, Tex=@n70) {samplerType=SAMPLERTYPE_Masks}.RGB
    B=MaterialFunctionCall(UV_Coord=@n69) {materialFunction=&MF_UV01_Mask}.Mask
   )) {group=Clock, maskChannel=Blue, parameterName=SignChannel, sortPriority=2, defaultValue=<rgba 0 0 1 0>}
  n72 = Clamp(None=ScalarParameter() {bUseCustomPrimitiveData=true, group=Clock, parameterName=TrainArrived, primitiveDataIndex=1})
  NamedRerouteDeclaration(None=LinearInterpolate(
    A=Add(A=@n61, B=@n66)
    B=SmoothStep(Min=@n67, Max=@n68, Value=@n71.R) {constMax=0.5, constMin=0.4}
    Alpha=NamedRerouteDeclaration(None=@n72) {name=ArrivalAlpha, nodeColor=<rgba 1 0 0.2578 1>}
   )) {name=Number Mask, nodeColor=<rgba 0.3359 1 0 1>}
  NamedRerouteDeclaration(None=Divide(A=Truncate!(None=@n7) {is=class MaterialExpressionTruncate}) {constB=6, desc=6X}) {name=MinuteProgressF1, nodeColor=<rgba 0.7344 1 0 1>}
  n73 = TextureObjectParameter() {
   group=Clock
   parameterName=NumberTexture
   samplerType=SAMPLERTYPE_Masks
   sortPriority=0
   texture=/Train/Props/Transit_ArrivalSigns_A/Textures/T_NumberComp
  }
  NamedRerouteDeclaration(None=@n73) {name=NumberTexture, nodeColor=<rgba 0 1 0.9062 1>}
  NamedRerouteDeclaration(None=@n8.Roughness) {name=RoughnessIN, nodeColor=<rgba 1 0.9609 0 1>}
  n74 = Constant() {r=60}
  n75 = Divide(
   A=NamedRerouteDeclaration(None=Fmod(A=@n4, B=@n74)) {name=SecondsRemainInAMinute, desc=This shows the seconds looping by 60, nodeColor=<rgba 0.3125 1 0 1>}
   B=Constant() {r=10}
  ) {desc=10X}
  NamedRerouteDeclaration(None=Frac(None=@n75)) {name=SecondsProgessF2, nodeColor=<rgba 1 0.1875 0 1>}
  NamedRerouteDeclaration(None=Divide(A=Truncate(None=@n75)) {constB=6, desc=6X}) {name=SecondsProgessF1, nodeColor=<rgba 1 0.1875 0 1>}
  TextureCoordinate()
 ]}
